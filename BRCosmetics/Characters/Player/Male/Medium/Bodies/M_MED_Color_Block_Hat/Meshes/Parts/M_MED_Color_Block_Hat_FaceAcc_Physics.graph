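# Physics asset for the face accessory mesh of the male medium Color Block Hat cosmetic, providing bodies and constraints for the hat bones.
# It chains three capsule bodies: a large one on head, a tiny one on dyn_hat_MSTR and a small offset one on dyn_hat.
# The MSTR-to-head joint leaves swing and twist free, while the dyn_hat-to-MSTR joint has tight angular limits of a few degrees.
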PhysicsAsset /BRCosmetics/Characters/Player/Male/Medium/Bodies/M_MED_Color_Block_Hat/Meshes/Parts/M_MED_Color_Block_Hat_FaceAcc_Physics
bodies {
 dyn_hat [{shapeType=Capsule, length=4.505, radius=2.38, shapeName=dyn_hat_capsule, center=<4.0902 0 0>, rotation=<rot 90 0 0>}]
 dyn_hat_MSTR [{shapeType=Capsule, length=0.505, radius=0.505, shapeName=dyn_hat_MSTR_capsule, rotation=<rot 0 -90 0>}]
 head [{
   shapeType=Capsule
   length=11.595
   radius=11.4629
   shapeName=head_capsule
   center=<7.0006 -0.5361 -0.0125>
   rotation=<rot 0 -43.8111 -89.6264>
  }]
}
constraints [
 {bone1Name=dyn_hat, bone2Name=dyn_hat_MSTR, swing1LimitDegrees=5, swing2LimitDegrees=8, twistLimitDegrees=5}
 {
  bone1Name=dyn_hat_MSTR
  bone2Name=head
  swing1LimitDegrees=45
  swing1Motion=Free
  swing2LimitDegrees=45
  swing2Motion=Free
  twistLimitDegrees=45
  twistMotion=Free
 }
]
